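# Physics asset for the InflateMate mic stand's backpack mesh in Sparks cosmetics, defining bodies for its bending stand and mic ribbon.
# Two capsule chains branch from a small handle sphere, stand bend 01–02 and 03–04, with tight angular limits for slight flex.
# A four-capsule ribbon chain hangs off stand bend 02 with much wider swing limits and locked twist, allowing looser secondary motion.
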PhysicsAsset /SparksCosmetics/Instrument/MicStand/InflateMate/Meshes/Backpack/InflateMate_MicStand_Backpack_Physics
bodies {
 Mic_dyn_ribbon_01 [{shapeType=Capsule, length=4, radius=0.7, shapeName=Mic_dyn_ribbon_01_capsule, center=<2.5 0 0>, rotation=<rot -90 0 0>}]
 Mic_dyn_ribbon_02 [{shapeType=Capsule, length=4, radius=0.7, shapeName=Mic_dyn_ribbon_02_capsule, center=<2.5 0 0>, rotation=<rot -90 0 0>}]
 Mic_dyn_ribbon_03 [{shapeType=Capsule, length=4, radius=0.7, shapeName=Mic_dyn_ribbon_03_capsule, center=<2.5 0 0>, rotation=<rot -90 0 0>}]
 Mic_dyn_ribbon_04 [{shapeType=Capsule, length=4, radius=0.7, shapeName=Mic_dyn_ribbon_04_capsule, center=<2.5 0 0>, rotation=<rot -90 0 0>}]
 Stand_dyn_stand_bend_01 [{shapeType=Capsule, length=16.5368, radius=3, shapeName=Stand_dyn_stand_bend_01_capsule, rotation=<rot 0 90 -90>}]
 Stand_dyn_stand_bend_02 [{shapeType=Capsule, length=27.6731, radius=3, shapeName=Stand_dyn_stand_bend_02_capsule, center=<6 0 1>, rotation=<rot 0 90 -90>}]
 Stand_dyn_stand_bend_03 [{shapeType=Capsule, length=11.8937, radius=4, shapeName=Stand_dyn_stand_bend_03_capsule, rotation=<rot 0 90 -90>}]
 Stand_dyn_stand_bend_04 [{shapeType=Capsule, length=23.6573, radius=4, shapeName=Stand_dyn_stand_bend_04_capsule, center=<3 0 0>, rotation=<rot 0 90 -90>}]
 handle [{shapeType=Sphere, radius=0.505, shapeName=handle_sphere}]
}
constraints [
 {bone1Name=Stand_dyn_stand_bend_03, bone2Name=handle, swing1LimitDegrees=8, swing2LimitDegrees=2, twistLimitDegrees=2, twistMotion=Locked}
 {bone1Name=Stand_dyn_stand_bend_04, bone2Name=Stand_dyn_stand_bend_03, swing1LimitDegrees=10, swing2LimitDegrees=2, twistLimitDegrees=2}
 {bone1Name=Stand_dyn_stand_bend_02, bone2Name=Stand_dyn_stand_bend_01, swing1LimitDegrees=5, swing2LimitDegrees=3, twistLimitDegrees=2}
 {bone1Name=Stand_dyn_stand_bend_01, bone2Name=handle, swing1LimitDegrees=7, swing2LimitDegrees=3, twistLimitDegrees=2}
 {
  bone1Name=Mic_dyn_ribbon_01
  bone2Name=Stand_dyn_stand_bend_02
  swing1LimitDegrees=60
  swing2LimitDegrees=35
  twistLimitDegrees=45
  twistMotion=Locked
 }
 {
  bone1Name=Mic_dyn_ribbon_02
  bone2Name=Mic_dyn_ribbon_01
  swing1LimitDegrees=35
  swing2LimitDegrees=35
  twistLimitDegrees=45
  twistMotion=Locked
 }
 {
  bone1Name=Mic_dyn_ribbon_03
  bone2Name=Mic_dyn_ribbon_02
  swing1LimitDegrees=30
  swing2LimitDegrees=20
  twistLimitDegrees=45
  twistMotion=Locked
 }
 {
  bone1Name=Mic_dyn_ribbon_04
  bone2Name=Mic_dyn_ribbon_03
  swing1LimitDegrees=35
  swing2LimitDegrees=20
  twistLimitDegrees=45
  twistMotion=Locked
 }
]
